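# Physics asset for the Summer Popsicle backpack cosmetic, defining simulated bodies and joints for its five "goop" bone chains.
# Each chain is two capsule bodies, a base constrained to the small pack capsule and an end constrained to that base.
# Angular limits are narrow throughout, with goop1's chain tightest and locked on its second swing axis, while the other four chains allow more sway and twist.
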PhysicsAsset /BRCosmetics/Accessories/FORT_Backpacks/Backpack_F_MED_Summer_Popsicle/Meshes/F_MED_Summer_Popsicle_Pack_Physics
bodies {
 goop1 [{shapeType=Capsule, length=3, radius=2.75, rotation=<rot 0 0 90>}]
 goop1_end [{shapeType=Capsule, length=3, radius=2.75, rotation=<rot 0 0 90>}]
 goop2 [{shapeType=Capsule, length=3, radius=2.75, rotation=<rot 0 0 90>}]
 goop2_end [{shapeType=Capsule, length=3, radius=2.75, rotation=<rot 0 0 90>}]
 goop3 [{shapeType=Capsule, length=3, radius=2.75, rotation=<rot 0 0 90>}]
 goop3_end [{shapeType=Capsule, length=3, radius=2.75, rotation=<rot 0 0 90>}]
 goop4 [{shapeType=Capsule, length=3, radius=2.75, rotation=<rot 0 0 90>}]
 goop4_end [{shapeType=Capsule, length=3, radius=2.75, rotation=<rot 0 0 90>}]
 goop5 [{shapeType=Capsule, length=5.2258, radius=2.3804, rotation=<rot 0 -90 0>}]
 goop5_end [{shapeType=Capsule, length=3.2258, radius=3.3804, rotation=<rot 0.3681 -90 30.1937>}]
 pack [{shapeType=Capsule, length=1, radius=1}]
}
constraints [
 {bone1Name=goop1, bone2Name=pack, swing1LimitDegrees=5, swing2LimitDegrees=2, swing2Motion=Locked, twistLimitDegrees=3}
 {bone1Name=goop1_end, bone2Name=goop1, swing1LimitDegrees=5, swing2LimitDegrees=2, swing2Motion=Locked, twistLimitDegrees=3}
 {bone1Name=goop2, bone2Name=pack, swing1LimitDegrees=15, swing2LimitDegrees=5, twistLimitDegrees=5}
 {bone1Name=goop2_end, bone2Name=goop2, swing1LimitDegrees=15, swing2LimitDegrees=15, twistLimitDegrees=10}
 {bone1Name=goop3, bone2Name=pack, swing1LimitDegrees=15, swing2LimitDegrees=5, twistLimitDegrees=5}
 {bone1Name=goop3_end, bone2Name=goop3, swing1LimitDegrees=15, swing2LimitDegrees=15, twistLimitDegrees=10}
 {bone1Name=goop4, bone2Name=pack, swing1LimitDegrees=15, swing2LimitDegrees=5, twistLimitDegrees=5}
 {bone1Name=goop4_end, bone2Name=goop4, swing1LimitDegrees=15, swing2LimitDegrees=15, twistLimitDegrees=10}
 {bone1Name=goop5, bone2Name=pack, swing1LimitDegrees=15, swing2LimitDegrees=5, twistLimitDegrees=5}
 {bone1Name=goop5_end, bone2Name=goop5, swing1LimitDegrees=15, swing2LimitDegrees=15, twistLimitDegrees=10}
]
