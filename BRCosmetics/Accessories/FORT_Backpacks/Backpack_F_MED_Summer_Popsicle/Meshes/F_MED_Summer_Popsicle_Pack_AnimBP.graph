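# Animation blueprint for the summer popsicle back bling, which morphs through stages over elapsed time, likely as melting.
# The event graph accumulates delta time, broadcasts it through a MaterialRoughness dispatcher, and after a delay blends in the Stage2 morph, then crossfades it to Stage3.
# The anim graph runs RigidBody physics on the ref pose, with alpha ramping up alongside the first morph phase.
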AnimBlueprint /BRCosmetics/Accessories/FORT_Backpacks/Backpack_F_MED_Summer_Popsicle/Meshes/F_MED_Summer_Popsicle_Pack_AnimBP
target_skeleton=/Game/Accessories/FORT_Backpacks/Backpack_F_MED_Summer_Popsicle/Meshes/F_MED_Summer_Popsicle_Pack_Skeleton
graphs [
 AnimGraph {nodes [
   n1 = RigidBody(ComponentPose=Local To Component(LocalPose=Local Space Ref Pose()), Alpha=Get morph_delta_alpha() -> (morph_delta_alpha:double)) {
    bOverrideWorldGravity=true
    overridePhysicsAsset=&/F_MED_Summer_Popsicle_Pack_Physics
    alphaBoolBlend {blendOption=Linear}
    baseBoneRef {boneName=root}
    componentAppliedLinearAccClamp=<2000 2000 2000>
    componentLinearAccScale=<0.3 0.3 0.35>
    componentLinearVelScale=<0.3 0.3 0.35>
    overrideWorldGravity {z=-980}
   }
   Output Pose(Result=Component To Local(ComponentPose=@n1)) {layerGroup=None}
  ]}
 EventGraph {nodes [
   n1 = Event BlueprintUpdateAnimation()
   > n2 = Set time(time:double=float + float(A=@n1.DeltaTimeX, B=Get time() -> (time:double)) {is=title K2Node_CommutativeAssociativeBinaryOperator}) -> (Output_Get:double)
   > Call Material Roughness!(execute:Exec, self:F MED Summer Popsicle Pack Anim BP Object Reference, Time:double=@n2) -> (then:Exec) {is=title K2Node_CallDelegate}
   > Set morph_delta_alpha(morph_delta_alpha:double=MapRangeClamped(Value=@n2, InRangeA=120, InRangeB=Get morph_half_range() -> (morph_half_range:double))) -> (Output_Get:double)
   > n3 = Branch(Condition="float < float"!(A:double=@n2, B:double=Get morph_half_range()) -> (ReturnValue:bool) {is=title K2Node_CallFunction})
   n4 = Get morph_half_range()
   > SetMorphTarget!(
    execute
    self:Anim Instance Object Reference
    MorphTargetName:Name=Popsicle_Stage2_LOD0Shape
    Value:float=MapRangeClamped(Value=@n2, InRangeA=120, InRangeB=@n4)
   ) -> (then)
   n5 = Get morph_full_range() -> (morph_full_range:double)
   SetMorphTarget(
    execute=@n3.else
    MorphTargetName=Popsicle_Stage2_LOD0Shape
    Value=MapRangeClamped(Value=@n2, InRangeA=@n4, InRangeB=@n5, OutRangeA=1, OutRangeB=0)
   )
   > SetMorphTarget(MorphTargetName=Popsicle_Stage3_LOD0Shape, Value=MapRangeClamped(Value=@n2, InRangeA=@n4, InRangeB=@n5))
   TryGetPawnOwner()
  ]}
 MaterialRoughness {nodes [MaterialRoughness!() -> (then, Time:double=~) {is=title K2Node_FunctionEntry}]}
]
variables [
 MaterialRoughness {default=()}
 morph_delta_alpha {type=number}
 morph_full_range {type=number, default=420}
 morph_half_range {type=number, default=270}
 time {type=number}
]
